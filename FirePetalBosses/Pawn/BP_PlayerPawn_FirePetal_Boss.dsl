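; Boss pawn for the FirePetal demon bosses, with empty stub functions for challenge start, spawn-cloud teleport, montages and last player leaving the arena.
; On authority, a sphere enabled after playlist load, backed by a looping failsafe timer, picks a non-downed participant as challenger and tags the boss challenged.
; An arena box reports when the last player leaves.
; Remotes spawn a grounded cloud and fade in audio; a RepNotify toggles mesh weapon collision.
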(event Custom|RestartChallengeCheckFailSafeTimer
  (bind _challengecheckfailsafeperiod_hf (Variables|Hotfix|GetChallengeCheckFailSafePeriod_Hf))
  (bind _returnvalue (ScalableFloat|GetValueatLevel _challengecheckfailsafeperiod_hf))
  ;[0.04589 0.04098 0.15 0.5] BeginPlay
  (if (> _returnvalue 0)
    (bind _returnvalue_2 (Utilities|Time|SetTimerbyEvent (EventDispatchers|CreateEvent :!event ChallengeCheckFailsafe) _returnvalue true))
    (Variables|Default|SetChallengeCheckFailSafeTimer _returnvalue_2)))

(event EventBeginPlay
  (bind _challengecheckfailsafeperiod_hf (Variables|Hotfix|GetChallengeCheckFailSafePeriod_Hf))
  (bind _returnvalue (ScalableFloat|GetValueatLevel _challengecheckfailsafeperiod_hf))
  (bind (_spawnlocation_x _spawnlocation_y) (Math|Vector|BreakVector (Variables|GetSpawnLocation_X)))
  (bind _bossarenabox (Variables|Default|GetBossArenaBox))
  ;[0.04589 0.04098 0.15 0.5] BeginPlay
  (|CallParentFunction)
  (Networking|SwitchHasAuthority
    (:Authority
      (Components|Sphere|SetSphereRadius (Variables|Default|GetChallengeCheckShere) (ScalableFloat|GetValueatLevel (Variables|Hotfix|GetChallengeCheckRadius_Hf)) false)
      (Variables|Default|SetSpawnLocation (Transformation|GetActorLocation))
      (Variables|Default|SetSpawnRotation (Transformation|GetActorRotation))
      (Class|FortAsyncActionCurrentPlaylistReady|PlaylistDataReadyAsync
        (:OnReady
          (Utilities|FlowControl|Delay 0.2)
          (Variables|/Script/Engine.PrimitiveComponent|SetbGenerateOverlapEvents true (Variables|Default|GetChallengeCheckShere))
          (Utilities|IsValid _bossarenabox
            (:Is%20Valid
              (Collision|BindEventtoOnActorEndOverlap _bossarenabox (EventDispatchers|CreateEvent :!event PlayerLeftBossArena)))
            (:Is%20Not%20Valid
              _))
          (if (> _returnvalue 0)
            (bind _returnvalue_2 (Utilities|Time|SetTimerbyEvent (EventDispatchers|CreateEvent :!event ChallengeCheckFailsafe) _returnvalue true))
            (Variables|Default|SetChallengeCheckFailSafeTimer _returnvalue_2)))))
    (:Remote
      (bind _returnvalue_3 (Game|SpawnActorfromClass "/FirePetalBosses/SpawnCloud/BP_FirePetalBosses_SpawnCloud.BP_FirePetalBosses_SpawnCloud_C" (Math|Transform|MakeTransform (Math|Vector|MakeVector _spawnlocation_x _spawnlocation_y (Class|AthenaFunctionLibrary|GetPlayerGroundedZ self))) "Undefined" "MultiplyWithRoot" 0 self))
      (Variables|Default|SetSpawnCloud _returnvalue_3)
      (bind _inairaudiocomp (Variables|Sounds|GetInAirAudioComp)
        (:then
          (Audio|Components|Audio|FadeIn _inairaudiocomp 1.0 1.0 0.0 "Linear" :!class /Script/Engine.AudioComponent))
        (:else)))))

(event Custom|ManeuverChosen
  (params (ManeuverTag :type Gameplay%20Tag%20Container%20Structure))
  ; Deprecated
  (if (== (Variables|Default|GetLastManeuverTag) ManeuverTag)
    (Math|Integer|IncrementInt (Variables|Default|GetAttackManeuverLoopCount))
    (else
      (Variables|Default|SetLastManeuverTag ManeuverTag)
      (Variables|Default|SetAttackManeuverLoopCount 0))))

(event OnComponentBeginOverlap(ChallengeCheckShere)
  (params (OverlappedComponent :type Primitive%20Component%20Object%20Reference) (OtherActor :type Actor%20Object%20Reference) (OtherComp :type Primitive%20Component%20Object%20Reference) (OtherBodyIndex :type Integer) (bFromSweep :type Boolean) (SweepResult :type Hit%20Result%20Structure%20%28by%20ref%29))
  ;[0.15 0.07916 0.07318 0.5] ChallengeCheckOverlap
  (bind _asfort_player_pawn (Utilities|Casting|CastToFortPlayerPawn OtherActor))
  (if (.bSuccess (Utilities|Casting|CastToBP_PlayerPawn_NonParticipant _asfort_player_pawn))
    _
    (elif (GameplayTags|HasMatchingGameplayTag "(TagName=\"FirePetalBosses.Demons.IsChallenged\")")
      _
      (elif (Health|IsDBNO _asfort_player_pawn)
        _
        (else
          (Variables|Default|SetChallengingPlayer _asfort_player_pawn)
          (Class|BPFLAICommonFunctions|AddUniqueGameplayTag self "(TagName=\"FirePetalBosses.Demons.IsChallenged\")"))))))

(event Custom|ChallengingPlayerFound
  ;[0.15 0.07916 0.07318 0.5] ChallengeCheckOverlap
  (Class|BPFLAICommonFunctions|AddUniqueGameplayTag self "(TagName=\"FirePetalBosses.Demons.IsChallenged\")"))

(event Custom|ChallengeCheckFailSafe
  ;[0.15 0.07916 0.07318 0.5] ChallengeCheckOverlap
  (Variables|Default|SetChallengingPlayer 0)
  (if (not (GameplayTags|HasMatchingGameplayTag "(TagName=\"FirePetalBosses.Demons.IsChallenged\")"))
    (bind _array_element (Utilities|Array|ForEachLoopwithBreak (Collision|GetOverlappingActors (Variables|Default|GetChallengeCheckShere) "/Script/FortniteGame.FortPlayerPawn" :!class /Script/Engine.PrimitiveComponent))
      (:LoopBody
        (if (.bSuccess (Utilities|Casting|CastToBP_PlayerPawn_NonParticipant _array_element))
          _
          (elif (Health|IsDBNO (Utilities|Casting|CastToFortPlayerPawn _array_element))
            _
            (else
              (Variables|Default|SetChallengingPlayer (Utilities|Casting|CastToFortPlayerPawn _array_element))
              (CallFunction|BreakChallengeCheckFailSafe)))))
      (:Completed
        (Utilities|IsValid (Variables|Default|GetChallengingPlayer)
          (:Is%20Valid
            (CallFunction|ChallengingPlayerFound)))))))

(event Custom|BreakChallengeCheckFailSafe
  ;[0.15 0.07916 0.07318 0.5] ChallengeCheckOverlap
  (bind _array_element (Utilities|Array|ForEachLoopwithBreak (Collision|GetOverlappingActors (Variables|Default|GetChallengeCheckShere) "/Script/FortniteGame.FortPlayerPawn" :!class /Script/Engine.PrimitiveComponent))
    (:LoopBody
      (if (.bSuccess (Utilities|Casting|CastToBP_PlayerPawn_NonParticipant _array_element))
        _
        (elif (Health|IsDBNO (Utilities|Casting|CastToFortPlayerPawn _array_element))
          _
          (else
            (Variables|Default|SetChallengingPlayer (Utilities|Casting|CastToFortPlayerPawn _array_element))
            (CallFunction|BreakChallengeCheckFailSafe)))))
    (:Completed
      (Utilities|IsValid (Variables|Default|GetChallengingPlayer)
        (:Is%20Valid
          (CallFunction|ChallengingPlayerFound))))))

(fn ChallengeStarted)

(fn InitiatedChallengeStart)

(fn InitiatedTeleportToSpawnCloud)

(fn LastPlayerLeftBossArena)

(fn OnRep_IgnoreWeaponCollision
  (CallFunction|UpdateMeshWeaponCollision (Variables|Default|GetIgnoreWeaponCollision)))

(fn PlayerLeftBossArena
  (params (OverlappedActor :type Actor%20Object%20Reference) (OtherActor :type Actor%20Object%20Reference))
  (if (Utilities|Array|IsEmpty (Collision|GetOverlappingActors (Variables|Default|GetBossArenaBox) "/Game/Athena/PlayerPawn_Athena.PlayerPawn_Athena_C"))
    (CallFunction|LastPlayerLeftBossArena)))

(fn PlayExitSCMontage
  (params (Montage :type Anim%20Montage%20Object%20Reference) (StartSection :type Name) (SkipAnim :type Boolean)))

(fn PlayWFCIdleMontage
  (params (Montage :type Anim%20Montage%20Object%20Reference) (StartSection :type Name)))

(fn TeleportedToSpawnCloud)

(fn UpdateMeshWeaponCollision
  (params (Ignore :type Boolean))
  (bind _mesh (Variables|Character|GetMesh))
  (Collision|SetCollisionResponsetoChannel _mesh "ECC_GameTraceChannel5" (select (Variables|GetIgnore) "ECR_Ignore" "ECR_Overlap"))
  (Collision|SetCollisionResponsetoChannel _mesh "ECC_GameTraceChannel7" (select (Variables|GetIgnore) "ECR_Ignore" "ECR_Block"))
  (Collision|SetCollisionResponsetoChannel _mesh "ECC_GameTraceChannel14" (select (Variables|GetIgnore) "ECR_Ignore" "ECR_Block"))
  (Collision|SetCollisionResponsetoChannel _mesh "ECC_GameTraceChannel16" (select (Variables|GetIgnore) "ECR_Ignore" "ECR_Overlap"))
  (Collision|SetCollisionResponsetoChannel _mesh "ECC_GameTraceChannel12" (select (Variables|GetIgnore) "ECR_Ignore" "ECR_Block")))

(fn ConstructionScript
  (|CallParentFunction))
